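# Looping, non-spatialized indoor ambience cue for large daytime interiors of the Freaky Fields haunted house POI.
# Every quality level plays a haunted hotel loop over a quieter secondary bed.
# The first level's bed crossfades pitch-randomized house and cave loops using opposing 100-second bell envelopes.
# The other levels share a cheaper bed: one house loop with random modulation and a 30-second looping pitch and volume drift.
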SoundCue /Game/Sounds/Helios/POI/FreakyFields/Ambience_Freaky_Fields_Haunted_House_Int_Day_Large_Cue
root=QualityLevel_0
nodes [
 {
  id=Enveloper_0
  children [WavePlayer_2]
  properties {
   bLoop=true
   bLoopIndefinitely=true
   loopEnd=30
   pitchCurve {editorCurveData {keys [
      {interpMode=RCIM_Cubic, time=0}
      {arriveTangent=0.0002, interpMode=RCIM_Cubic, leaveTangent=0.0002, tangentMode=RCTM_User, time=7.9875, value=1.04}
      {arriveTangent=0, interpMode=RCIM_Cubic, leaveTangent=0, tangentMode=RCTM_User, time=20.0737, value=0.92}
      {interpMode=RCIM_Cubic, time=30}
     ]}}
   volumeCurve {editorCurveData {keys [<at 0.9111 0.8713 RCIM_Cubic>,<at 16.565 0.7143 RCIM_Cubic>,<at 30.9111 0.8713 RCIM_Cubic>]}}
  }
 }
 {
  id=Enveloper_18
  children [Modulator_18]
  properties {
   bLoop=true
   bLoopIndefinitely=true
   loopEnd=100
   pitchCurve {editorCurveData {keys [{time=0}]}}
   volumeCurve {externalCurve=/Game/Sounds/Trans_Enveloper_Curves/Bell_100s, editorCurveData {keys [<at 0 1>]}}
  }
 }
 {
  id=Enveloper_19
  children [Modulator_19]
  properties {
   bLoop=true
   bLoopIndefinitely=true
   loopEnd=100
   pitchCurve {editorCurveData {keys [{time=0}]}}
   volumeCurve {externalCurve=/Game/Sounds/Trans_Enveloper_Curves/BellReversed_100s.BellReversed_100S, editorCurveData {keys [<at 0 1>]}}
  }
 }
 {id=Mixer_0, children [WavePlayer_0,Mixer_8], properties {inputVolume [1,0.6]}}
 {id=Mixer_1, children [Modulator_0,WavePlayer_0], properties {inputVolume [1,0.6]}}
 {id=Mixer_8, children [Enveloper_18,Enveloper_19], properties {inputVolume [1,1]}}
 {id=Modulator_0, children [Enveloper_0], properties {pitchMax=1, pitchMin=0.8, volumeMax=0.9, volumeMin=0.85}}
 {id=Modulator_18, children [WavePlayer_1], properties {pitchMax=1, pitchMin=0.8, volumeMax=1.1, volumeMin=0.9}}
 {id=Modulator_19, children [WavePlayer_14], properties {pitchMax=1.3, pitchMin=1.1, volumeMax=1.1, volumeMin=0.9}}
 {id=QualityLevel_0, children [Mixer_0,Mixer_1,Mixer_1,Mixer_1]}
 {
  id=WavePlayer_0
  properties {bLooping=true, soundWaveAssetPtr=&/AMBI_Starlight_mToD_Haunted_Hotel_Int_Day_Large_Loop}
 }
 {id=WavePlayer_1, properties {bLooping=true, soundWaveAssetPtr=&/SmallHouseA007_Loop_Stereo}}
 {id=WavePlayer_14, properties {bLooping=true, soundWaveAssetPtr=&/Cave_Low07}}
 {id=WavePlayer_2, properties {bLooping=true, soundWaveAssetPtr=&SmallHouseA007_Loop_Stereo}}
]
properties {
 bOverrideConcurrency=true
 maxDistance=2097152
 soundClassObject=&Fort_Ambient_Indoors_Athena
 volumeMultiplier=0.251
 attenuationOverrides {bAttenuate=false, bEnableReverbSend=false, bSpatialize=false}
 concurrencyOverrides {resolutionRule=PreventNew}
}
