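# Sound cue for the looping night ambience inside the Freaky Fields haunted house basement.
# A single looping wave player plays a reused Starlight haunted hotel basement night loop at reduced volume through the indoor ambient sound class.
# Attenuation overrides disable distance falloff, spatialization and reverb send, so it plays as a non-positional bed, and a concurrency override prevents new instances once the limit is reached.
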SoundCue /Game/Sounds/Helios/POI/FreakyFields/Ambience_Freaky_Fields_Haunted_House_Int_Night_Basement_Cue
root=WavePlayer_0
nodes [{
  id=WavePlayer_0
  properties {bLooping=true, soundWaveAssetPtr=&/AMBI_Starlight_mToD_Haunted_Hotel_Int_Night_Basement_Loop}
 }]
properties {
 bOverrideConcurrency=true
 maxDistance=2097152
 soundClassObject=&Fort_Ambient_Indoors_Athena
 volumeMultiplier=0.251
 attenuationOverrides {bAttenuate=false, bEnableReverbSend=false, bSpatialize=false}
 concurrencyOverrides {resolutionRule=PreventNew}
}
