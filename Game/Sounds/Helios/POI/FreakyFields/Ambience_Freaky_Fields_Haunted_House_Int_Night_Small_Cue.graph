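# Looping, non-spatialized Sound Cue for small haunted-house interior night ambience in Freaky Fields.
# A quality switch sends its first level to a base loop mixed with two randomly modulated house layers, faded by opposing 100-second bell and reversed-bell envelopes.
# The other three levels share a simpler mix: one house loop with a 30-second pitch and volume drift, plus a duplicate envelope branch with no wave connected.
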SoundCue /Game/Sounds/Helios/POI/FreakyFields/Ambience_Freaky_Fields_Haunted_House_Int_Night_Small_Cue
root=QualityLevel_0
nodes [
 {
  id=Enveloper_0
  children [WavePlayer_2]
  properties {
   bLoop=true
   bLoopIndefinitely=true
   loopEnd=30
   pitchCurve {editorCurveData {keys [
      {interpMode=RCIM_Cubic, time=0}
      {arriveTangent=0.0002, interpMode=RCIM_Cubic, leaveTangent=0.0002, tangentMode=RCTM_User, time=7.9875, value=1.04}
      {arriveTangent=0, interpMode=RCIM_Cubic, leaveTangent=0, tangentMode=RCTM_User, time=20.0737, value=0.92}
      {interpMode=RCIM_Cubic, time=30}
     ]}}
   volumeCurve {editorCurveData {keys [<at 0 1 RCIM_Cubic>,<at 15.6539 0.8429 RCIM_Cubic>,<at 30 1 RCIM_Cubic>]}}
  }
 }
 {
  id=Enveloper_1
  children [null]
  properties {
   bLoop=true
   bLoopIndefinitely=true
   loopEnd=30
   pitchCurve {editorCurveData {keys [
      {interpMode=RCIM_Cubic, time=0}
      {arriveTangent=0.0002, interpMode=RCIM_Cubic, leaveTangent=0.0002, tangentMode=RCTM_User, time=7.9875, value=1.04}
      {arriveTangent=0, interpMode=RCIM_Cubic, leaveTangent=0, tangentMode=RCTM_User, time=20.0737, value=0.92}
      {interpMode=RCIM_Cubic, time=30}
     ]}}
   volumeCurve {editorCurveData {keys [<at 0 1 RCIM_Cubic>,<at 15.6539 0.8429 RCIM_Cubic>,<at 30 1 RCIM_Cubic>]}}
  }
 }
 {
  id=Enveloper_34
  children [Modulator_34]
  properties {
   bLoop=true
   bLoopIndefinitely=true
   loopEnd=100
   pitchCurve {editorCurveData {keys [{time=0}]}}
   volumeCurve {externalCurve=/Game/Sounds/Trans_Enveloper_Curves/Bell_100s, editorCurveData {keys [<at 0 1>]}}
  }
 }
 {
  id=Enveloper_35
  children [Modulator_35]
  properties {
   bLoop=true
   bLoopIndefinitely=true
   loopEnd=100
   pitchCurve {editorCurveData {keys [{time=0}]}}
   volumeCurve {externalCurve=/Game/Sounds/Trans_Enveloper_Curves/BellReversed_100s.BellReversed_100S, editorCurveData {keys [<at 0 1>]}}
  }
 }
 {id=Mixer_0, children [WavePlayer_0,Mixer_16], properties {inputVolume [1,0.6]}}
 {id=Mixer_1, children [Modulator_0,Modulator_1], properties {inputVolume [1,1]}}
 {id=Mixer_16, children [Enveloper_34,Enveloper_35], properties {inputVolume [1,0.5]}}
 {id=Modulator_0, children [Enveloper_0], properties {pitchMax=1.1, pitchMin=0.85, volumeMax=1, volumeMin=0.9}}
 {id=Modulator_1, children [Enveloper_1], properties {pitchMax=1.1, pitchMin=0.85, volumeMax=1, volumeMin=0.9}}
 {id=Modulator_34, children [WavePlayer_29], properties {pitchMax=1, pitchMin=0.8, volumeMax=1.1, volumeMin=0.9}}
 {id=Modulator_35, children [WavePlayer_1], properties {pitchMax=1.3, pitchMin=1.1, volumeMax=1.1, volumeMin=0.9}}
 {id=QualityLevel_0, children [Mixer_0,Mixer_1,Mixer_1,Mixer_1]}
 {
  id=WavePlayer_0
  properties {bLooping=true, soundWaveAssetPtr=&/AMBI_Starlight_mToD_Haunted_Hotel_Int_Night_Small_Loop}
 }
 {id=WavePlayer_1, properties {bLooping=true, soundWaveAssetPtr=&/SmallHouseA018_Loop_Stereo}}
 {id=WavePlayer_2, properties {bLooping=true, soundWaveAssetPtr=&SmallHouseA018_Loop_Stereo}}
 {id=WavePlayer_29, properties {bLooping=true, soundWaveAssetPtr=&/SmallHouseB003_Stereo}}
]
properties {
 bOverrideConcurrency=true
 maxDistance=2097152
 soundClassObject=&Fort_Ambient_Indoors_Athena
 volumeMultiplier=0.251
 attenuationOverrides {bAttenuate=false, bEnableReverbSend=false, bSpatialize=false}
 concurrencyOverrides {resolutionRule=PreventNew}
}
